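# DodgyBean body-fall foley cue, layering a random Robert body-fall wave with a random, pitch-varied Burt heavy gallop wave.
# IsLocal and IsRiddenByEnemy branches select among three mixes of these layers, quieter when neither local nor enemy-ridden.
# A quality-level node gives the first level eight body-fall variants and the others a three-wave subset.
# The cue is pitched down, uses heavy predator-step attenuation and overrides concurrency to cap instances.
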SoundCue /Irwin/Sounds/DodgyBean/DodgyBean_Foley_BodyFall_Cue
root=Branch_1
nodes [
 {id=Branch_0, children [Mixer_1,Mixer_0,Mixer_1], properties {boolParameterName=IsRiddenByEnemy}}
 {id=Branch_1, children [Mixer_2,Branch_0,Branch_0], properties {boolParameterName=IsLocal}}
 {id=Mixer_0, children [QualityLevel_0, Modulator_3], properties {inputVolume [0.3,0.4]}}
 {id=Mixer_1, children [QualityLevel_0, Modulator_3], properties {inputVolume [1,1]}}
 {id=Mixer_2, children [QualityLevel_0, Modulator_3], properties {inputVolume [0.7,1.5]}}
 {id=Modulator_3, children [Random_0], properties {pitchMax=1, pitchMin=0.9}}
 {id=QualityLevel_0, children [Random_4,Random_5,Random_5,Random_5]}
 {id=Random_0, children [WavePlayer_24, WavePlayer_25, WavePlayer_26, WavePlayer_27, WavePlayer_28], properties {weights [1,1,1,1,1]}}
 {
  id=Random_4
  children [WavePlayer_0,WavePlayer_1,WavePlayer_2,WavePlayer_3,WavePlayer_4,WavePlayer_5,WavePlayer_6,WavePlayer_7]
  properties {weights [1,1,1,1,1,1,1,1]}
 }
 {id=Random_5, children [WavePlayer_5,WavePlayer_6,WavePlayer_7], properties {weights [1,1,1]}}
 {id=WavePlayer_0, properties {soundWaveAssetPtr=/Irwin/Sounds/Robert/Foley/Robert_Foley_BodyFall_01}}
 {id=WavePlayer_1, properties {soundWaveAssetPtr=/Irwin/Sounds/Robert/Foley/Robert_Foley_BodyFall_02}}
 {id=WavePlayer_2, properties {soundWaveAssetPtr=/Irwin/Sounds/Robert/Foley/Robert_Foley_BodyFall_03}}
 {id=WavePlayer_24, properties {soundWaveAssetPtr=/Irwin/Sounds/Burt/Foley/Burt_Gallup_Heavy_01}}
 {id=WavePlayer_25, properties {soundWaveAssetPtr=/Irwin/Sounds/Burt/Foley/Burt_Gallup_Heavy_02}}
 {id=WavePlayer_26, properties {soundWaveAssetPtr=/Irwin/Sounds/Burt/Foley/Burt_Gallup_Heavy_03}}
 {id=WavePlayer_27, properties {soundWaveAssetPtr=/Irwin/Sounds/Burt/Foley/Burt_Gallup_Heavy_04}}
 {id=WavePlayer_28, properties {soundWaveAssetPtr=/Irwin/Sounds/Burt/Foley/Burt_Gallup_Heavy_05}}
 {id=WavePlayer_3, properties {soundWaveAssetPtr=/Irwin/Sounds/Robert/Foley/Robert_Foley_BodyFall_04}}
 {id=WavePlayer_4, properties {soundWaveAssetPtr=/Irwin/Sounds/Robert/Foley/Robert_Foley_BodyFall_05}}
 {id=WavePlayer_5, properties {soundWaveAssetPtr=/Irwin/Sounds/Robert/Foley/Robert_Foley_BodyFall_06}}
 {id=WavePlayer_6, properties {soundWaveAssetPtr=/Irwin/Sounds/Robert/Foley/Robert_Foley_BodyFall_07}}
 {id=WavePlayer_7, properties {soundWaveAssetPtr=/Irwin/Sounds/Robert/Foley/Robert_Foley_BodyFall_08}}
]
properties {
 attenuationSettings=/Irwin/Sounds/Settings/Predator_Step_Heavy_Att
 bOverrideConcurrency=true
 maxDistance=6000
 pitchMultiplier=0.8
 soundClassObject=/Game/Sounds/FortSoundClassesAndModes/Fort_Creatures_Foley
 virtualizationMode=Disabled
 volumeMultiplier=1
 concurrencyOverrides {maxCount=6}
}
