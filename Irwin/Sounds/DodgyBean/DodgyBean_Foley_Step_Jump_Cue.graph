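# Jump footstep cue for DodgyBean, built from Robert run and jump foley with heavy predator step attenuation and a lowered pitch.
# IsLocal selects the local mix; otherwise IsRiddenByEnemy picks between the full mix and a half-volume, run-only mix for non-enemy riders or none.
# Full mixes layer random run steps with modulated jump sounds.
# Quality-level nodes use eight run variations at the first level and three at the others.
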SoundCue /Irwin/Sounds/DodgyBean/DodgyBean_Foley_Step_Jump_Cue
root=Branch_1
nodes [
 {id=Branch_0, children [Mixer_1,Mixer_0,Mixer_1], properties {boolParameterName=IsRiddenByEnemy}}
 {id=Branch_1, children [Mixer_2,Branch_0,Branch_0], properties {boolParameterName=IsLocal}}
 {id=Mixer_0, children [QualityLevel_2], properties {inputVolume [0.5]}}
 {id=Mixer_1, children [QualityLevel_0, Modulator_4], properties {inputVolume [1,1.2]}}
 {id=Mixer_2, children [QualityLevel_1, Modulator_3], properties {inputVolume [2,1]}}
 {id=Modulator_3, children [Random_6], properties {pitchMax=1.2, volumeMin=0.3}}
 {id=Modulator_4, children [Random_7]}
 {id=QualityLevel_0, children [Random_1,Random_3,Random_3,Random_3]}
 {id=QualityLevel_1, children [Random_0,Random_4,Random_4,Random_4]}
 {id=QualityLevel_2, children [Random_2,Random_5,Random_5,Random_5]}
 {
  id=Random_0
  children [WavePlayer_0,WavePlayer_1,WavePlayer_2,WavePlayer_3,WavePlayer_4,WavePlayer_5,WavePlayer_6,WavePlayer_7]
  properties {bShouldExcludeFromBranchCulling=true, weights [1,1,1,1,1,1,1,1]}
 }
 {
  id=Random_1
  children [WavePlayer_8, WavePlayer_9, WavePlayer_10, WavePlayer_11, WavePlayer_12, WavePlayer_13, WavePlayer_14, WavePlayer_15]
  properties {bShouldExcludeFromBranchCulling=true, weights [1,1,1,1,1,1,1,1]}
 }
 {
  id=Random_2
  children [WavePlayer_16, WavePlayer_17, WavePlayer_18, WavePlayer_19, WavePlayer_20, WavePlayer_21, WavePlayer_22, WavePlayer_23]
  properties {bShouldExcludeFromBranchCulling=true, weights [1,1,1,1,1,1,1,1]}
 }
 {id=Random_3, children [WavePlayer_13, WavePlayer_14, WavePlayer_15], properties {weights [1,1,1]}}
 {id=Random_4, children [WavePlayer_5,WavePlayer_6,WavePlayer_7], properties {weights [1,1,1]}}
 {id=Random_5, children [WavePlayer_21, WavePlayer_22, WavePlayer_23], properties {weights [1,1,1]}}
 {
  id=Random_6
  children [WavePlayer_36, WavePlayer_37, WavePlayer_38, WavePlayer_39]
  properties {bShouldExcludeFromBranchCulling=true, weights [1,1,1,1]}
 }
 {
  id=Random_7
  children [WavePlayer_40, WavePlayer_41, WavePlayer_42, WavePlayer_43]
  properties {bShouldExcludeFromBranchCulling=true, weights [1,1,1,1]}
 }
 {id=WavePlayer_0, properties {soundWaveAssetPtr=/Irwin/Sounds/Robert/Foley/Robert_Foley_Footstep_Run_01}}
 {id=WavePlayer_1, properties {soundWaveAssetPtr=/Irwin/Sounds/Robert/Foley/Robert_Foley_Footstep_Run_02}}
 {id=WavePlayer_10, properties {soundWaveAssetPtr=/Irwin/Sounds/Robert/Foley/Robert_Foley_Footstep_Run_03}}
 {id=WavePlayer_11, properties {soundWaveAssetPtr=/Irwin/Sounds/Robert/Foley/Robert_Foley_Footstep_Run_04}}
 {id=WavePlayer_12, properties {soundWaveAssetPtr=/Irwin/Sounds/Robert/Foley/Robert_Foley_Footstep_Run_05}}
 {id=WavePlayer_13, properties {soundWaveAssetPtr=/Irwin/Sounds/Robert/Foley/Robert_Foley_Footstep_Run_06}}
 {id=WavePlayer_14, properties {soundWaveAssetPtr=/Irwin/Sounds/Robert/Foley/Robert_Foley_Footstep_Run_07}}
 {id=WavePlayer_15, properties {soundWaveAssetPtr=/Irwin/Sounds/Robert/Foley/Robert_Foley_Footstep_Run_08}}
 {id=WavePlayer_16, properties {soundWaveAssetPtr=&Robert_Foley_Footstep_Run_01}}
 {id=WavePlayer_17, properties {soundWaveAssetPtr=&Robert_Foley_Footstep_Run_02}}
 {id=WavePlayer_18, properties {soundWaveAssetPtr=&Robert_Foley_Footstep_Run_03}}
 {id=WavePlayer_19, properties {soundWaveAssetPtr=&Robert_Foley_Footstep_Run_04}}
 {id=WavePlayer_2, properties {soundWaveAssetPtr=&Robert_Foley_Footstep_Run_03}}
 {id=WavePlayer_20, properties {soundWaveAssetPtr=&Robert_Foley_Footstep_Run_05}}
 {id=WavePlayer_21, properties {soundWaveAssetPtr=&Robert_Foley_Footstep_Run_06}}
 {id=WavePlayer_22, properties {soundWaveAssetPtr=&Robert_Foley_Footstep_Run_07}}
 {id=WavePlayer_23, properties {soundWaveAssetPtr=&Robert_Foley_Footstep_Run_08}}
 {id=WavePlayer_3, properties {soundWaveAssetPtr=&Robert_Foley_Footstep_Run_04}}
 {id=WavePlayer_36, properties {soundWaveAssetPtr=/Irwin/Sounds/Robert/Foley/Robert_Foley_Footstep_Jump_01}}
 {id=WavePlayer_37, properties {soundWaveAssetPtr=/Irwin/Sounds/Robert/Foley/Robert_Foley_Footstep_Jump_02}}
 {id=WavePlayer_38, properties {soundWaveAssetPtr=/Irwin/Sounds/Robert/Foley/Robert_Foley_Footstep_Jump_03}}
 {id=WavePlayer_39, properties {soundWaveAssetPtr=/Irwin/Sounds/Robert/Foley/Robert_Foley_Footstep_Jump_04}}
 {id=WavePlayer_4, properties {soundWaveAssetPtr=&Robert_Foley_Footstep_Run_05}}
 {id=WavePlayer_40, properties {soundWaveAssetPtr=&Robert_Foley_Footstep_Jump_01}}
 {id=WavePlayer_41, properties {soundWaveAssetPtr=&Robert_Foley_Footstep_Jump_02}}
 {id=WavePlayer_42, properties {soundWaveAssetPtr=&Robert_Foley_Footstep_Jump_03}}
 {id=WavePlayer_43, properties {soundWaveAssetPtr=&Robert_Foley_Footstep_Jump_04}}
 {id=WavePlayer_5, properties {soundWaveAssetPtr=&Robert_Foley_Footstep_Run_06}}
 {id=WavePlayer_6, properties {soundWaveAssetPtr=&Robert_Foley_Footstep_Run_07}}
 {id=WavePlayer_7, properties {soundWaveAssetPtr=&Robert_Foley_Footstep_Run_08}}
 {id=WavePlayer_8, properties {soundWaveAssetPtr=&Robert_Foley_Footstep_Run_01}}
 {id=WavePlayer_9, properties {soundWaveAssetPtr=&Robert_Foley_Footstep_Run_02}}
]
properties {
 attenuationSettings=/Irwin/Sounds/Settings/Predator_Step_Heavy_Att
 bOverrideConcurrency=true
 maxDistance=6000
 pitchMultiplier=0.9
 soundClassObject=/Game/Sounds/FortSoundClassesAndModes/Fort_Creatures_Steps_Normal
 virtualizationMode=Disabled
 volumeMultiplier=1
 concurrencyOverrides {maxCount=6}
}
