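# Landing footstep foley cue for the DodgyBean creature, built from borrowed Robert and Burt sounds under heavy predator step attenuation and creature-step sound class.
# An IsLocal branch plays body falls layered with two pitched-up heavy gallop layers.
# Otherwise, IsRiddenByEnemy selects body falls plus land footsteps when true or unset, or quieter body falls alone when false.
# Quality-level nodes trim random variations on lower tiers, and concurrency is capped.
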SoundCue /Irwin/Sounds/DodgyBean/DodgyBean_Foley_Step_Land_Cue
root=Branch_1
nodes [
 {id=Branch_0, children [Mixer_1,Mixer_0,Mixer_1], properties {boolParameterName=IsRiddenByEnemy}}
 {id=Branch_1, children [Mixer_2,Branch_0,Branch_0], properties {boolParameterName=IsLocal}}
 {id=Mixer_0, children [QualityLevel_2], properties {inputVolume [0.5]}}
 {id=Mixer_1, children [QualityLevel_2, Modulator_1], properties {inputVolume [1.25,1.25]}}
 {id=Mixer_2, children [QualityLevel_1, Modulator_3, Modulator_2], properties {inputVolume [1.75,2,2]}}
 {id=Modulator_1, children [QualityLevel_5]}
 {id=Modulator_2, children [QualityLevel_3], properties {pitchMax=2, pitchMin=2}}
 {id=Modulator_3, children [QualityLevel_4], properties {pitchMax=2.5, pitchMin=2.5}}
 {id=QualityLevel_1, children [Random_0,Random_0,Random_7,Random_7]}
 {id=QualityLevel_2, children [Random_1,Random_1,Random_8,Random_8]}
 {id=QualityLevel_3, children [Random_4, Random_4, WavePlayer_41, WavePlayer_41]}
 {id=QualityLevel_4, children [Random_3, Random_3, WavePlayer_28, WavePlayer_28]}
 {id=QualityLevel_5, children [Random_5, Random_5, WavePlayer_34, WavePlayer_34]}
 {
  id=Random_0
  children [WavePlayer_0,WavePlayer_1,WavePlayer_2,WavePlayer_3,WavePlayer_4,WavePlayer_5,WavePlayer_6,WavePlayer_7]
  properties {bShouldExcludeFromBranchCulling=true, weights [1,1,1,1,1,1,1,1]}
 }
 {
  id=Random_1
  children [WavePlayer_8, WavePlayer_9, WavePlayer_10, WavePlayer_11, WavePlayer_12, WavePlayer_13, WavePlayer_14, WavePlayer_15]
  properties {bShouldExcludeFromBranchCulling=true, weights [1,1,1,1,1,1,1,1]}
 }
 {id=Random_3, children [WavePlayer_24, WavePlayer_25, WavePlayer_26, WavePlayer_27, WavePlayer_28], properties {weights [1,1,1,1,1]}}
 {id=Random_4, children [WavePlayer_29, WavePlayer_30, WavePlayer_31, WavePlayer_40, WavePlayer_41], properties {weights [1,1,1,1,1]}}
 {
  id=Random_5
  children [WavePlayer_32, WavePlayer_33, WavePlayer_34, WavePlayer_35]
  properties {bShouldExcludeFromBranchCulling=true, weights [1,1,1,1]}
 }
 {id=Random_7, children [WavePlayer_5,WavePlayer_6,WavePlayer_7], properties {weights [1,1,1]}}
 {id=Random_8, children [WavePlayer_13, WavePlayer_14, WavePlayer_15], properties {weights [1,1,1]}}
 {id=WavePlayer_0, properties {soundWaveAssetPtr=/Irwin/Sounds/Robert/Foley/Robert_Foley_BodyFall_01}}
 {id=WavePlayer_1, properties {soundWaveAssetPtr=/Irwin/Sounds/Robert/Foley/Robert_Foley_BodyFall_02}}
 {id=WavePlayer_10, properties {soundWaveAssetPtr=/Irwin/Sounds/Robert/Foley/Robert_Foley_BodyFall_03}}
 {id=WavePlayer_11, properties {soundWaveAssetPtr=/Irwin/Sounds/Robert/Foley/Robert_Foley_BodyFall_04}}
 {id=WavePlayer_12, properties {soundWaveAssetPtr=/Irwin/Sounds/Robert/Foley/Robert_Foley_BodyFall_05}}
 {id=WavePlayer_13, properties {soundWaveAssetPtr=/Irwin/Sounds/Robert/Foley/Robert_Foley_BodyFall_06}}
 {id=WavePlayer_14, properties {soundWaveAssetPtr=/Irwin/Sounds/Robert/Foley/Robert_Foley_BodyFall_07}}
 {id=WavePlayer_15, properties {soundWaveAssetPtr=/Irwin/Sounds/Robert/Foley/Robert_Foley_BodyFall_08}}
 {id=WavePlayer_2, properties {soundWaveAssetPtr=&Robert_Foley_BodyFall_03}}
 {id=WavePlayer_24, properties {soundWaveAssetPtr=/Irwin/Sounds/Burt/Foley/Burt_Gallup_Heavy_01}}
 {id=WavePlayer_25, properties {soundWaveAssetPtr=/Irwin/Sounds/Burt/Foley/Burt_Gallup_Heavy_02}}
 {id=WavePlayer_26, properties {soundWaveAssetPtr=/Irwin/Sounds/Burt/Foley/Burt_Gallup_Heavy_03}}
 {id=WavePlayer_27, properties {soundWaveAssetPtr=/Irwin/Sounds/Burt/Foley/Burt_Gallup_Heavy_04}}
 {id=WavePlayer_28, properties {soundWaveAssetPtr=/Irwin/Sounds/Burt/Foley/Burt_Gallup_Heavy_05}}
 {id=WavePlayer_29, properties {soundWaveAssetPtr=&Burt_Gallup_Heavy_01}}
 {id=WavePlayer_3, properties {soundWaveAssetPtr=&Robert_Foley_BodyFall_04}}
 {id=WavePlayer_30, properties {soundWaveAssetPtr=&Burt_Gallup_Heavy_02}}
 {id=WavePlayer_31, properties {soundWaveAssetPtr=&Burt_Gallup_Heavy_03}}
 {id=WavePlayer_32, properties {soundWaveAssetPtr=/Irwin/Sounds/Robert/Foley/Robert_Foley_Footstep_Land_01}}
 {id=WavePlayer_33, properties {soundWaveAssetPtr=/Irwin/Sounds/Robert/Foley/Robert_Foley_Footstep_Land_02}}
 {id=WavePlayer_34, properties {soundWaveAssetPtr=/Irwin/Sounds/Robert/Foley/Robert_Foley_Footstep_Land_03}}
 {id=WavePlayer_35, properties {soundWaveAssetPtr=/Irwin/Sounds/Robert/Foley/Robert_Foley_Footstep_Land_04}}
 {id=WavePlayer_4, properties {soundWaveAssetPtr=&Robert_Foley_BodyFall_05}}
 {id=WavePlayer_40, properties {soundWaveAssetPtr=&Burt_Gallup_Heavy_04}}
 {id=WavePlayer_41, properties {soundWaveAssetPtr=&Burt_Gallup_Heavy_05}}
 {id=WavePlayer_5, properties {soundWaveAssetPtr=&Robert_Foley_BodyFall_06}}
 {id=WavePlayer_6, properties {soundWaveAssetPtr=&Robert_Foley_BodyFall_07}}
 {id=WavePlayer_7, properties {soundWaveAssetPtr=&Robert_Foley_BodyFall_08}}
 {id=WavePlayer_8, properties {soundWaveAssetPtr=&Robert_Foley_BodyFall_01}}
 {id=WavePlayer_9, properties {soundWaveAssetPtr=&Robert_Foley_BodyFall_02}}
]
properties {
 attenuationSettings=/Irwin/Sounds/Settings/Predator_Step_Heavy_Att
 bOverrideConcurrency=true
 maxDistance=6000
 soundClassObject=/Game/Sounds/FortSoundClassesAndModes/Fort_Creatures_Steps_Normal
 virtualizationMode=Disabled
 volumeMultiplier=1
 concurrencyOverrides {maxCount=6}
}
