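# DodgyBean running footstep foley cue for a creature, with heavy predator-step attenuation and owner-limited concurrency.
# A shared core layers randomized Robert run footsteps, whose pool size varies by quality level, with Burt heavy gallop samples pitched up 2.5x.
# IsLocal and IsEnemy branches route that core through different gains, keeping enemies loudest, local steps moderately reduced and non-enemies quietest.
# An outer modulator adds slight pitch and volume variation.
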SoundCue /Irwin/Sounds/DodgyBean/DodgyBean_Foley_Step_Run_Cue
root=Modulator_0
nodes [
 {id=Branch_0, children [Mixer_1,Mixer_0,Mixer_1], properties {boolParameterName=IsEnemy}}
 {id=Branch_1, children [Mixer_2,Branch_0,Branch_0], properties {boolParameterName=IsLocal}}
 {id=Mixer_0, children [Modulator_2], properties {inputVolume [0.3]}}
 {id=Mixer_1, children [Mixer_3], properties {inputVolume [1]}}
 {id=Mixer_2, children [Modulator_1], properties {inputVolume [0.7]}}
 {id=Mixer_3, children [QualityLevel_0, Modulator_3], properties {inputVolume [1,1]}}
 {id=Modulator_0, children [Branch_1], properties {pitchMax=1.1, pitchMin=0.9, volumeMin=0.85}}
 {id=Modulator_1, children [Mixer_3], properties {pitchMax=1, pitchMin=1, volumeMax=1.25, volumeMin=0.5}}
 {id=Modulator_2, children [Mixer_3], properties {pitchMax=1, pitchMin=1, volumeMax=1.25, volumeMin=1}}
 {id=Modulator_3, children [Random_5], properties {pitchMax=2.5, pitchMin=2.5}}
 {id=QualityLevel_0, children [Random_2,Random_3,Random_3,Random_3]}
 {
  id=Random_2
  children [WavePlayer_16, WavePlayer_17, WavePlayer_18, WavePlayer_19, WavePlayer_20, WavePlayer_21, WavePlayer_22, WavePlayer_23]
  properties {bShouldExcludeFromBranchCulling=true, weights [1,1,1,1,1,1,1,1]}
 }
 {id=Random_3, children [WavePlayer_21, WavePlayer_22, WavePlayer_23], properties {weights [1,1,1]}}
 {id=Random_5, children [WavePlayer_24, WavePlayer_25, WavePlayer_26, WavePlayer_27, WavePlayer_28], properties {weights [1,1,1,1,1]}}
 {id=WavePlayer_16, properties {soundWaveAssetPtr=/Irwin/Sounds/Robert/Foley/Robert_Foley_Footstep_Run_01}}
 {id=WavePlayer_17, properties {soundWaveAssetPtr=/Irwin/Sounds/Robert/Foley/Robert_Foley_Footstep_Run_02}}
 {id=WavePlayer_18, properties {soundWaveAssetPtr=/Irwin/Sounds/Robert/Foley/Robert_Foley_Footstep_Run_03}}
 {id=WavePlayer_19, properties {soundWaveAssetPtr=/Irwin/Sounds/Robert/Foley/Robert_Foley_Footstep_Run_04}}
 {id=WavePlayer_20, properties {soundWaveAssetPtr=/Irwin/Sounds/Robert/Foley/Robert_Foley_Footstep_Run_05}}
 {id=WavePlayer_21, properties {soundWaveAssetPtr=/Irwin/Sounds/Robert/Foley/Robert_Foley_Footstep_Run_06}}
 {id=WavePlayer_22, properties {soundWaveAssetPtr=/Irwin/Sounds/Robert/Foley/Robert_Foley_Footstep_Run_07}}
 {id=WavePlayer_23, properties {soundWaveAssetPtr=/Irwin/Sounds/Robert/Foley/Robert_Foley_Footstep_Run_08}}
 {id=WavePlayer_24, properties {soundWaveAssetPtr=/Irwin/Sounds/Burt/Foley/Burt_Gallup_Heavy_01}}
 {id=WavePlayer_25, properties {soundWaveAssetPtr=/Irwin/Sounds/Burt/Foley/Burt_Gallup_Heavy_02}}
 {id=WavePlayer_26, properties {soundWaveAssetPtr=/Irwin/Sounds/Burt/Foley/Burt_Gallup_Heavy_03}}
 {id=WavePlayer_27, properties {soundWaveAssetPtr=/Irwin/Sounds/Burt/Foley/Burt_Gallup_Heavy_04}}
 {id=WavePlayer_28, properties {soundWaveAssetPtr=/Irwin/Sounds/Burt/Foley/Burt_Gallup_Heavy_05}}
]
properties {
 attenuationSettings=/Irwin/Sounds/Settings/Predator_Step_Heavy_Att
 bOverrideConcurrency=true
 maxDistance=6000
 pitchMultiplier=0.9
 soundClassObject=/Game/Sounds/FortSoundClassesAndModes/Fort_Creatures_Steps_Normal
 volumeMultiplier=1
 concurrencyOverrides {bLimitToOwner=true, maxCount=2}
}
