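# Sound cue for DodgyBean's toe-pick foley, apparently from an idle break animation, played as a creature step sound with heavy predator step attenuation.
# IsLocal and IsRiddenByEnemy branches route to separate mixers, all currently at equal volume and feeding one shared quality switch.
# The first quality level randomly picks from five waves, the others from a three-wave subset.
# The cue lowers pitch and overrides concurrency to cap simultaneous instances.
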SoundCue /Irwin/Sounds/DodgyBean/DodgyBean_Foley_Toe_Pick_Cue
root=Branch_1
nodes [
 {id=Branch_0, children [Mixer_1,Mixer_0,Mixer_1], properties {boolParameterName=IsRiddenByEnemy}}
 {id=Branch_1, children [Mixer_2,Branch_0,Branch_0], properties {boolParameterName=IsLocal}}
 {id=Mixer_0, children [QualityLevel_1], properties {inputVolume [1]}}
 {id=Mixer_1, children [QualityLevel_1], properties {inputVolume [1]}}
 {id=Mixer_2, children [QualityLevel_1], properties {inputVolume [1]}}
 {id=QualityLevel_1, children [Random_3,Random_4,Random_4,Random_4]}
 {id=Random_3, children [WavePlayer_21, WavePlayer_24, WavePlayer_25, WavePlayer_26, WavePlayer_27], properties {weights [1,1,1,1,1]}}
 {id=Random_4, children [WavePlayer_25, WavePlayer_26, WavePlayer_27], properties {weights [1,1,1]}}
 {id=WavePlayer_21, properties {soundWaveAssetPtr=/Irwin/Sounds/DodgyBean/Wavs/Robert_Idle_Break1_Foley_Toe_Pick_01}}
 {id=WavePlayer_24, properties {soundWaveAssetPtr=/Irwin/Sounds/DodgyBean/Wavs/Robert_Idle_Break1_Foley_Toe_Pick_04}}
 {id=WavePlayer_25, properties {soundWaveAssetPtr=/Irwin/Sounds/DodgyBean/Wavs/Robert_Idle_Break1_Foley_Toe_Pick_05}}
 {id=WavePlayer_26, properties {soundWaveAssetPtr=/Irwin/Sounds/DodgyBean/Wavs/Robert_Idle_Break1_Foley_Toe_Pick_06}}
 {id=WavePlayer_27, properties {soundWaveAssetPtr=/Irwin/Sounds/DodgyBean/Wavs/Robert_Idle_Break1_Foley_Toe_Pick_07}}
]
properties {
 attenuationSettings=/Irwin/Sounds/Settings/Predator_Step_Heavy_Att
 bOverrideConcurrency=true
 maxDistance=6000
 pitchMultiplier=0.8
 soundClassObject=/Game/Sounds/FortSoundClassesAndModes/Fort_Creatures_Steps_Normal
 virtualizationMode=Disabled
 volumeMultiplier=1
 concurrencyOverrides {maxCount=6}
}
