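# Sound cue for the DodgyBean melee bite teeth chomp, playing one of four equally weighted Robert chomp foley waves at lowered pitch.
# An IsLocal branch and a nested IsRiddenByEnemy branch route through separate mixers into one shared Random node, all currently at equal volume.
# It uses Predator bite attenuation, the tactical creature vocalization sound class and a concurrency override that caps simultaneous instances.
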SoundCue /Irwin/Sounds/DodgyBean/DodgyBean_Melee_Bite_Teeth_Chomp_Cue
root=Branch_1
nodes [
 {id=Branch_0, children [Mixer_1,Mixer_0,Mixer_1], properties {boolParameterName=IsRiddenByEnemy}}
 {id=Branch_1, children [Mixer_2,Branch_0,Branch_0], properties {boolParameterName=IsLocal}}
 {id=Mixer_0, children [Random_0], properties {inputVolume [1]}}
 {id=Mixer_1, children [Random_0], properties {inputVolume [1]}}
 {id=Mixer_2, children [Random_0], properties {inputVolume [1]}}
 {
  id=Random_0
  children [WavePlayer_0,WavePlayer_1,WavePlayer_2,WavePlayer_3]
  properties {bShouldExcludeFromBranchCulling=true, weights [1,1,1,1]}
 }
 {id=WavePlayer_0, properties {soundWaveAssetPtr=/Irwin/Sounds/Robert/Foley/Robert_Melee_Bite_Teeth_Chomp_04}}
 {id=WavePlayer_1, properties {soundWaveAssetPtr=/Irwin/Sounds/Robert/Foley/Robert_Melee_Bite_Teeth_Chomp_01}}
 {id=WavePlayer_2, properties {soundWaveAssetPtr=/Irwin/Sounds/Robert/Foley/Robert_Melee_Bite_Teeth_Chomp_02}}
 {id=WavePlayer_3, properties {soundWaveAssetPtr=/Irwin/Sounds/Robert/Foley/Robert_Melee_Bite_Teeth_Chomp_03}}
]
properties {
 attenuationSettings=/Irwin/Sounds/Settings/Predator_Bite_Att
 bOverrideConcurrency=true
 maxDistance=2000
 pitchMultiplier=0.8
 soundClassObject=&Fort_Creatures_Vocalization_Tactical
 volumeMultiplier=1
 concurrencyOverrides {maxCount=3}
}
