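# DodgyBean bite sound cue that plays one of several wolf bite attack impact sounds.
# A random node with equal weights picks from four wave variations.
# The cue is slightly raised in pitch and volume, uses a predator bite attenuation, and is routed to the tactical creature vocalization sound class.
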SoundCue /Irwin/Sounds/DodgyBean/IP_SC_DodgyBean_Bite
root=Random_0
nodes [
 {id=Random_0, children [WavePlayer_0,WavePlayer_1,WavePlayer_2,WavePlayer_3], properties {weights [1,1,1,1]}}
 {id=WavePlayer_0, properties {soundWaveAssetPtr=/Irwin/Sounds/DodgyBean/Wavs/Wolf_Bite_Attack_Impact_01}}
 {id=WavePlayer_1, properties {soundWaveAssetPtr=/Irwin/Sounds/DodgyBean/Wavs/Wolf_Bite_Attack_Impact_02}}
 {id=WavePlayer_2, properties {soundWaveAssetPtr=/Irwin/Sounds/DodgyBean/Wavs/Wolf_Bite_Attack_Impact_03}}
 {id=WavePlayer_3, properties {soundWaveAssetPtr=/Irwin/Sounds/DodgyBean/Wavs/Wolf_Bite_Attack_Impact_04}}
]
properties {
 attenuationSettings=/Irwin/Sounds/Settings/Predator_Bite_Att
 maxDistance=2000
 pitchMultiplier=1.2
 soundClassObject=&Fort_Creatures_Vocalization_Tactical
 volumeMultiplier=1.2
}
