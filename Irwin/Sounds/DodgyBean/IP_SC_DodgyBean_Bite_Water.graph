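# Sound cue for the DodgyBean creature's bite, named as the water variant.
# A random node picks among three equally weighted bear attack bite foley waves.
# The cue raises pitch and volume, uses the medium predator vocalization attenuation and routes to the large creatures sound class.
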SoundCue /Irwin/Sounds/DodgyBean/IP_SC_DodgyBean_Bite_Water
root=Random_0
nodes [
 {id=Random_0, children [WavePlayer_0,WavePlayer_1,WavePlayer_2], properties {weights [1,1,1]}}
 {id=WavePlayer_0, properties {soundWaveAssetPtr=/Irwin/Sounds/DodgyBean/Wavs/CRTR_Bear_Foley_Attack_Bite_01}}
 {id=WavePlayer_1, properties {soundWaveAssetPtr=/Irwin/Sounds/DodgyBean/Wavs/CRTR_Bear_Foley_Attack_Bite_02}}
 {id=WavePlayer_2, properties {soundWaveAssetPtr=/Irwin/Sounds/DodgyBean/Wavs/CRTR_Bear_Foley_Attack_Bite_03}}
]
properties {
 attenuationSettings=/Irwin/Sounds/Settings/Predator_Vocalization_Medium_Att
 maxDistance=12000
 pitchMultiplier=1.5
 soundClassObject=/Game/Sounds/FortSoundClassesAndModes/Fort_Creatures_Large
 volumeMultiplier=1.2
}
