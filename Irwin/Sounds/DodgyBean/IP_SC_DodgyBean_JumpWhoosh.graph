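# DodgyBean jump whoosh Sound Cue, using Fortnite player foley class and third-person jump attenuation.
# An IsLocal branch gives the local player a mix of random wall-jump kick-off and VividRazor Baron whooshes.
# Non-local playback is pitched down further and routed through IsEnemy, playing only kick-off whooshes for enemies and silence for non-enemies.
# Unset parameters fall back to the kick-off variations, and the whole cue is pitched down.
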SoundCue /Irwin/Sounds/DodgyBean/IP_SC_DodgyBean_JumpWhoosh
root=Branch_0
nodes [
 {id=Branch_0, children [Mixer_0,Modulator_0,Branch_1], properties {boolParameterName=IsLocal}}
 {id=Branch_1, children [Random_0,null,Random_0], properties {boolParameterName=IsEnemy}}
 {id=Mixer_0, children [Random_0,Random_1], properties {inputVolume [1,1]}}
 {id=Modulator_0, children [Branch_1], properties {pitchMax=0.8, pitchMin=0.8}}
 {id=Random_0, children [WavePlayer_0,WavePlayer_1,WavePlayer_2,WavePlayer_3], properties {weights [1,1,1,1]}}
 {id=Random_1, children [WavePlayer_4,WavePlayer_5], properties {weights [1,1]}}
 {id=WavePlayer_0, properties {soundWaveAssetPtr=/Irwin/Sounds/DodgyBean/Wavs/Wall_Jump_Kick_Off_Whoosh_04}}
 {id=WavePlayer_1, properties {soundWaveAssetPtr=/Irwin/Sounds/DodgyBean/Wavs/Wall_Jump_Kick_Off_Whoosh_03}}
 {id=WavePlayer_2, properties {soundWaveAssetPtr=/Irwin/Sounds/DodgyBean/Wavs/Wall_Jump_Kick_Off_Whoosh_02}}
 {id=WavePlayer_3, properties {soundWaveAssetPtr=/Irwin/Sounds/DodgyBean/Wavs/Wall_Jump_Kick_Off_Whoosh_01}}
 {id=WavePlayer_4, properties {soundWaveAssetPtr=/Irwin/Sounds/DodgyBean/Wavs/VividRazor_Baron_JumpWhoosh_02}}
 {id=WavePlayer_5, properties {soundWaveAssetPtr=/Irwin/Sounds/DodgyBean/Wavs/VividRazor_Baron_JumpWhoosh_01}}
]
properties {
 attenuationSettings=/Game/Sounds/Attenuation/Fort_Player_Jump_3P_Athena_Att
 maxDistance=3584
 pitchMultiplier=0.8
 priority=2
 soundClassObject=&Fort_Player_Foley_Athena
 volumeMultiplier=1
}
